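# Control bus fader widget for audio modulation utilities: a vertical slider from -60 to 0 dB with a rotated bus name label and a numeric level readout.
# SetFaderData stores the channel index, sets the name, and converts a normalized value to dB (20·log10) for slider and readout.
# On Construct it binds slider changes, which broadcast OnFaderValueChanged with linear gain (10^(dB/20)) and channel, then update the readout.
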WidgetBlueprint /Game/Sounds/Fort_Audio_Blueprints/Utilities/EditorUtilityWidgets/Modulation/WBP_ControlBusFader
graphs [
 EventGraph {nodes [
   n1 = OnValueChanged!() -> (OutputDelegate, then, Value:float=~) {is=title K2Node_CustomEvent}
   > Call On Fader Value Changed!(
    execute:Exec
    self:WBP Control Bus Fader Object Reference
    Value:double=Power!(Base:double=10, Exp:double=float / float(A=@n1.Value, B=20)) -> (ReturnValue:double) {is=title K2Node_CallFunction}
    Channel:int=Get Fader Channel() -> (Fader Channel:int=0)
   ) -> (then:Exec) {is=title K2Node_CallDelegate}
   > SetCurrentValue!(
    execute
    self:Common Numeric Text Block Object Reference=Get VolumeLevelText() -> (VolumeLevelText:Common Numeric Text Block Object Reference)
    NewValue:float=@n1.Value
   ) -> (then)
   Event Construct()
   > Bind Event to On Value Changed(self:Slider Object Reference=Get AnalogSlider_51() -> (AnalogSlider_51:Analog Slider Object Reference), Delegate=@n1) {is=title K2Node_AddDelegate}
  ]}
 OnFaderValueChanged {nodes [OnFaderValueChanged!() -> (then, Value:double=~, Channel:int) {is=title K2Node_FunctionEntry}]}
 SetFaderData {nodes [
   n1 = SetFaderData!() -> (then, FaderChannel:int, NormalizedValue:double=~, Fader Name:Text) {is=title K2Node_FunctionEntry}
   > Set Fader Channel(Fader Channel:int=@n1) -> (Output_Get:int=0)
   > Set Text(Text=@n1.Fader Name, self=Get BusName() -> (BusName:Common Text Object Reference))
   n2 = float * float(A=Log!(A:double=@n1.NormalizedValue, Base:double=10) -> (ReturnValue:double), B=20) {is=title K2Node_PromotableOperator}
   > SetCurrentValue(self=Get VolumeLevelText(), NewValue=@n2)
   > SetValue!(execute, self:Slider Object Reference=Get AnalogSlider_51(), InValue:float=@n2) -> (then)
  ]}
]
variables [
 Fader Channel {type=integer}
 OnFaderValueChanged {}
]
widgets {
 info {inheritedWidgetCount=0, parentClass=&UserWidget, rootWidgetClass=&CanvasPanel, widgetCount=9}
 widgets [
  CanvasPanel_41 {widgetClassPath=&CanvasPanel, widget=CanvasPanel_41}
  SizeBox_79 {
   widgetClassPath=&SizeBox
   parent=CanvasPanel_41
   widget=SizeBox_79
   properties {bOverride_HeightOverride=true, heightOverride=500}
   slot {
    class=CanvasPanelSlot
    properties {layoutData {alignment=<0 0>, anchors {maximum=<0 0>, minimum=<0 0>}, offsets=<ltrb 20 0 33.9339 368.9103>}}
   }
  }
  ScaleBox_84 {
   widgetClassPath=&ScaleBox
   parent=SizeBox_79
   widget=ScaleBox_84
   properties {stretch=Fill, renderTransform {angle=0, scale=<2 1>, shear=<0 0>, translation=<0 0>}}
   slot {class=SizeBoxSlot, properties {horizontalAlignment=HAlign_Center}}
  }
  AnalogSlider_51 {
   widgetClassPath=/Script/CommonUI.AnalogSlider
   bIsVariable=true
   parent=ScaleBox_84
   widget=AnalogSlider_51
   properties {
    maxValue=0
    minValue=-60
    orientation=Orient_Vertical
    sliderBarColor=<rgba 0.7917 0.7917 0.7917 1>
    sliderHandleColor=<rgba 0.3021 0.3021 0.3021 1>
    widgetStyle {
     barThickness=8.6068
     disabledBarImage {
      bIsDynamicallyLoaded=false
      drawAs=Image
      imageType=NoImage
      mirroring=NoMirror
      resourceName=None
      resourceObject=null
      tiling=NoTile
      imageSize=<0 0>
      margin=<ltrb 0 0 0 0>
      outlineSettings {
       bUseBrushTransparency=false
       roundingType=HalfHeightRadius
       width=0
       color {colorUseRule=UseColor_Specified, specifiedColor=<rgba 0 0 0 0>}
       cornerRadii=<0 0 0 0>
      }
      tintColor {colorUseRule=UseColor_Specified, specifiedColor=<rgba 0.5 0.5 0.5 1>}
      uVRegion {bIsValid=false, max=<0 0>, min=<0 0>}
     }
     disabledThumbImage {
      bIsDynamicallyLoaded=false
      drawAs=RoundedBox
      imageType=FullColor
      mirroring=NoMirror
      resourceName=../../../Engine/Content/Slate/Common/Button.png
      resourceObject=null
      tiling=NoTile
      imageSize=<36 14>
      margin=<ltrb 0 0 0 0>
      outlineSettings {
       bUseBrushTransparency=false
       roundingType=HalfHeightRadius
       width=0
       color {colorUseRule=UseColor_Specified, specifiedColor=<rgba 0 0 0 0>}
       cornerRadii=<0 0 0 0>
      }
      tintColor {colorUseRule=UseColor_Specified, specifiedColor=<rgba 1 1 1 1>}
      uVRegion {bIsValid=false, max=<0 0>, min=<0 0>}
     }
     hoveredBarImage {
      bIsDynamicallyLoaded=false
      drawAs=Image
      imageType=NoImage
      mirroring=NoMirror
      resourceName=None
      resourceObject=null
      tiling=NoTile
      imageSize=<0 0>
      margin=<ltrb 0 0 0 0>
      outlineSettings {
       bUseBrushTransparency=false
       roundingType=HalfHeightRadius
       width=0
       color {colorUseRule=UseColor_Specified, specifiedColor=<rgba 0 0 0 0>}
       cornerRadii=<0 0 0 0>
      }
      tintColor {colorUseRule=UseColor_Specified, specifiedColor=<rgba 1 1 1 1>}
      uVRegion {bIsValid=false, max=<0 0>, min=<0 0>}
     }
     hoveredThumbImage {
      bIsDynamicallyLoaded=false
      drawAs=RoundedBox
      imageType=FullColor
      mirroring=NoMirror
      resourceName=../../../Engine/Content/Slate/Common/Button.png
      resourceObject=null
      tiling=NoTile
      imageSize=<36 14>
      margin=<ltrb 0 0 0 0>
      outlineSettings {
       bUseBrushTransparency=false
       roundingType=HalfHeightRadius
       width=0
       color {colorUseRule=UseColor_Specified, specifiedColor=<rgba 0 0 0 0>}
       cornerRadii=<0 0 0 0>
      }
      tintColor {colorUseRule=UseColor_Specified, specifiedColor=<rgba 1 1 1 1>}
      uVRegion {bIsValid=false, max=<0 0>, min=<0 0>}
     }
     normalBarImage {
      bIsDynamicallyLoaded=false
      drawAs=Image
      imageType=NoImage
      mirroring=NoMirror
      resourceName=None
      resourceObject=null
      tiling=NoTile
      imageSize=<0 0>
      margin=<ltrb 0 0 0 0>
      outlineSettings {
       bUseBrushTransparency=false
       roundingType=HalfHeightRadius
       width=0
       color {colorUseRule=UseColor_Specified, specifiedColor=<rgba 0 0 0 0>}
       cornerRadii=<0 0 0 0>
      }
      tintColor {colorUseRule=UseColor_Specified, specifiedColor=<rgba 1 1 1 1>}
      uVRegion {bIsValid=false, max=<0 0>, min=<0 0>}
     }
     normalThumbImage {
      bIsDynamicallyLoaded=false
      drawAs=RoundedBox
      imageType=FullColor
      mirroring=NoMirror
      resourceName=../../../Engine/Content/Slate/Common/Button.png
      resourceObject=null
      tiling=NoTile
      imageSize=<36 14>
      margin=<ltrb 0 0 0 0>
      outlineSettings {
       bUseBrushTransparency=false
       roundingType=HalfHeightRadius
       width=0
       color {colorUseRule=UseColor_Specified, specifiedColor=<rgba 0 0 0 0>}
       cornerRadii=<0 0 0 0>
      }
      tintColor {colorUseRule=UseColor_Specified, specifiedColor=<rgba 1 1 1 1>}
      uVRegion {bIsValid=false, max=<0 0>, min=<0 0>}
     }
    }
   }
   slot {class=ScaleBoxSlot}
  }
  SizeBox_59 {
   widgetClassPath=&SizeBox
   parent=CanvasPanel_41
   widget=SizeBox_59
   slot {
    class=CanvasPanelSlot
    properties {layoutData {alignment=<0 0>, anchors {maximum=<0 0>, minimum=<0 0>}, offsets=<ltrb 8 72 15.6429 279.1583>}}
   }
  }
  BusName {
   widgetClassPath=&CommonTextBlock
   bIsVariable=true
   parent=SizeBox_59
   widget=BusName
   properties {
    autoWrapText=true
    bIsScrollingEnabled=false
    style=/FortUILibrary/Styles/Typography/_TextStyle_TagExtraSmall._TextStyle_TagExtraSmall_C
    text=Name
    font {
     bForceMonospaced=false
     bMaterialIsStencil=false
     fontMaterial=null
     fontObject=&Burbank
     letterSpacing=0
     monospacedWidth=1
     size=12
     skewAmount=0
     typefaceFontName=Burbank-PlayerPower
     outlineSettings {
      bApplyOutlineToDropShadows=false
      bMiteredCorners=false
      bSeparateFillAlpha=false
      outlineMaterial=null
      outlineSize=0
      outlineColor=<rgba 0 0 0 1>
     }
    }
    renderTransform {angle=-90, scale=<1 1>, shear=<0 0>, translation=<0 0>}
    shadowOffset=<0 0>
    strikeBrush {
     bIsDynamicallyLoaded=false
     drawAs=NoDrawType
     imageType=NoImage
     mirroring=NoMirror
     resourceName=None
     resourceObject=null
     tiling=NoTile
     imageSize=<32 32>
     margin=<ltrb 0 0 0 0>
     outlineSettings {
      bUseBrushTransparency=false
      roundingType=HalfHeightRadius
      width=0
      color {colorUseRule=UseColor_Specified, specifiedColor=<rgba 0 0 0 0>}
      cornerRadii=<0 0 0 0>
     }
     tintColor {colorUseRule=UseColor_Specified, specifiedColor=<rgba 1 1 1 1>}
     uVRegion {bIsValid=false, max=<0 0>, min=<0 0>}
    }
   }
   slot {class=SizeBoxSlot, properties {horizontalAlignment=HAlign_Right, verticalAlignment=VAlign_Bottom}}
  }
  SizeBox_120 {
   widgetClassPath=&SizeBox
   parent=CanvasPanel_41
   widget=SizeBox_120
   slot {
    class=CanvasPanelSlot
    properties {layoutData {alignment=<0 0>, anchors {maximum=<0 0>, minimum=<0 0>}, offsets=<ltrb 8 356 43.6937 31.6892>}}
   }
  }
  ScaleBox_47 {
   widgetClassPath=&ScaleBox
   parent=SizeBox_120
   widget=ScaleBox_47
   slot {class=SizeBoxSlot, properties {horizontalAlignment=HAlign_Center, verticalAlignment=VAlign_Top}}
  }
  VolumeLevelText {
   widgetClassPath=/Script/CommonUI.CommonNumericTextBlock
   bIsVariable=true
   parent=ScaleBox_47
   widget=VolumeLevelText
   properties {
    style=/FortUILibrary/Styles/Typography/Foundation/TextStyle-Power-S.TextStyle-Power-S_C
    text=Text Block
    font {
     bForceMonospaced=false
     bMaterialIsStencil=false
     fontMaterial=null
     fontObject=&Burbank
     letterSpacing=0
     monospacedWidth=1
     size=18
     skewAmount=0
     typefaceFontName=Burbank-PlayerPower
     outlineSettings {
      bApplyOutlineToDropShadows=false
      bMiteredCorners=false
      bSeparateFillAlpha=false
      outlineMaterial=null
      outlineSize=0
      outlineColor=<rgba 0 0 0 1>
     }
    }
    formattingSpecification {
     alwaysSign=false
     maximumFractionalDigits=0
     maximumIntegralDigits=2
     minimumFractionalDigits=2
     minimumIntegralDigits=1
     roundingMode=HalfFromZero
     useGrouping=true
    }
    shadowOffset=<0 0>
   }
   slot {class=ScaleBoxSlot, properties {horizontalAlignment=HAlign_Fill, verticalAlignment=VAlign_Fill}}
  }
 ]
}
