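# Editor utility widget for audio designers that mixes a Sound Control Bus Mix with on-screen faders and a MIDI fader controller (PreSonus FP8 by default).
# It builds one fader per mix stage, activating the mix in the PIE or editor world.
# Fader moves map through dB curves and update the mix in transactions, echoing to motor faders via pitch bend.
# MIDI notes shift banks of eight faders or save the asset.
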EditorUtilityWidgetBlueprint /Game/Sounds/Fort_Audio_Blueprints/Utilities/EditorUtilityWidgets/Modulation/EUW_ControlBusMixer
graphs [
 ChannelToFaderId {nodes [
   n1 = ChannelToFaderId!() -> (then, A:int) {is=title K2Node_FunctionEntry}
   > Return Node!(execute, ReturnValue:int="% (Integer)"(A=@n1, B=8) -> (ReturnValue=0))
  ]}
 EventGraph {nodes [
   "On Property Changed (ControlBusMix_Property)"!() -> (OutputDelegate, then, PropertyName:Name=None) {is=title K2Node_ComponentBoundEvent}
   > SetObject(
    self=Get DetailsView() -> (DetailsView:Details View Object Reference)
    NewObject=Get BusMixToUse() -> (BusMixToUse:Sound Control Bus Mix Object Reference)
   )
   > Activate Control Bus Mix!(
    execute
    WorldContextObject:Object Reference=GetCurrentWorld!() -> (ReturnValue:World Object Reference)
    Mix:Sound Control Bus Mix Object Reference=Get BusMixToUse()
   ) -> (then)
   > SetUpFaders()
   > ResetAllFaders()
   > InitializePhysicalFaders()
   n1 = "On Property Changed (DetailsView)"!() -> (OutputDelegate, then, PropertyName:Name=None) {is=title K2Node_ComponentBoundEvent}
   > Branch(Condition=OR Boolean(A="Equal (Name)"(A=@n1.PropertyName, B=Bus), B="Equal (Name)"(A=@n1.PropertyName, B=MixStages)))
   > PrintString(InString=Append(A="To String (Name)"(InName=@n1.PropertyName), B=" Changed - rebuilding faders"), bPrintToScreen=false)
   > SetUpFaders()
   > ResetAllFaders()
   > InitializePhysicalFaders()
   "On Clicked (Button_Refresh)"() {is=title K2Node_ComponentBoundEvent}
   > SetUpFaders()
   > InitializePhysicalFaders()
   n2 = OnMIDIPitchBend_Event!() -> (OutputDelegate, then, MIDIDeviceController:Object Reference, Timestamp:int, Channel:int, Pitch:int) {is=title K2Node_CustomEvent}
   > UpdateFader!(execute, self, Channel:int=@n2.Channel, Pitch:int=@n2.Pitch) -> (then)
   n3 = On Midi Note On!() -> (OutputDelegate, then, MIDIDeviceController:Object Reference, Timestamp:int, Channel:int, Note:int, Velocity:int) {is=title K2Node_CustomEvent}
   > PrintString(InString="To String (Text)"(InText=Format Text!(Format="LOCTEXT(\"Midi Note channel {C} ## {N}\")", C:int=@n3.Channel, N:int=@n3.Note) -> (Result)))
   > n4 = Branch(Condition="Equal (Integer)"(A=@n3.Note, B=94))
   > SaveControlBus()
   n5 = Branch(execute=@n4.else, Condition="Equal (Integer)"(A=@n3.Note, B=46))
   > UpdateFaderShift!(execute, self, IsShiftUp:bool=false) -> (then)
   Branch(execute=@n5.else, Condition="Equal (Integer)"(A=@n3.Note, B=47))
   > UpdateFaderShift(IsShiftUp=true)
   Event Construct()
   > Activate Control Bus Mix(Mix=Get BusMixToUse())
   > FindInputDevices()
   > n6 = Create MIDIDevice Input Controller!(execute, DeviceID:int=Get InDeviceID() -> (InDeviceID:int=0), MIDIBufferSize:int=1024) -> (then, ReturnValue:Object Reference)
   > Set MIDI Input Controller(MIDI Input Controller:Object Reference=@n6) -> (Output_Get:Object Reference)
   n7 = Get MIDI Input Controller() -> (MIDI Input Controller:Object Reference)
   > "Assign <invalid delegate>"!(execute:Exec, self:Object Reference=@n7, Delegate:"Delegate (by ref)"=@n2) -> (then:Exec) {is=title K2Node_AssignDelegate}
   > Bind Event to OnMIDINoteOn(self:Object Reference=@n7, Delegate=@n3) {is=title K2Node_AddDelegate}
   > SetObject(self=Get DetailsView(), NewObject=Get BusMixToUse())
   > n8 = Create MIDIDevice Output Controller!(execute, DeviceID:int=Get OutDeviceID() -> (OutDeviceID:int=0)) -> (then, ReturnValue:Object Reference)
   > Set MIDI Output Controller(MIDI Output Controller:Object Reference=@n8) -> (Output_Get:Object Reference)
   > SetUpFaders()
   > InitializePhysicalFaders()
   n9 = Get DeviceName!(self:Object Reference=Get MIDI Input Controller()) -> (DeviceName:String)
   n10 = Get DeviceName(self=Get MIDI Output Controller() -> (MIDI Output Controller:Object Reference))
   > PrintString(InString="To String (Text)"(InText=Format Text!(Format="LOCTEXT(\"Input Device {I} and output device {O} created\")", I:String=@n9, O:String=@n10) -> (Result)))
   Event Tick()
   > Sequence()
   > n11 = Delta Gate Wildcard!(
    execute:Exec
    Variable:"Object Reference (by ref)"=Get CurrentWorld() -> (CurrentWorld:Object Reference)
    New Value:Object Reference=GetCurrentWorld()
   ) -> (Success:Exec, Updated Variable:"Object Reference (by ref)", Previous Variable:Object Reference) {is=title K2Node_MacroInstance}
   > Activate Control Bus Mix(WorldContextObject=@n11, Mix=Get BusMixToUse())
   Event PreConstruct()
   > SetObject(self=Get ControlBusMix_Property() -> (ControlBusMix_Property:Single Property View Object Reference), NewObject=Self-Reference())
   > SetObject(self=Get DetailsView(), NewObject=Get BusMixToUse())
   > SetObject(self=Get DeviceName_Property() -> (DeviceName_Property:Single Property View Object Reference), NewObject=Self-Reference())
   Event Destruct()
   > Set MIDI Input Controller()
   > Set MIDI Output Controller()
   > Shut Down All MIDIDevices!(execute) -> (then)
   > CollectGarbage!(execute) -> (then)
   AddChild(self=Get ScrollBox_40() -> (ScrollBox_40:Scroll Box Object Reference), Content=Get DetailsView())
  ]}
 FaderIdToMixStage {nodes [
   n1 = FaderIdToMixStage!() -> (then, A:int) {is=title K2Node_FunctionEntry}
   n2 = Get FaderShiftIdx() -> (FaderShiftIdx:int=0)
   > Return Node(ReturnValue=int + int(A=@n1, B=int * int(A=@n2, B=8)))
  ]}
 FaderPitchToLinearVolume {nodes [
   n1 = FaderPitchToLinearVolume!() -> (then, InInt:int) {is=title K2Node_FunctionEntry}
   n2 = GetFloatValue!(
    self:Curve Float Object Reference=Get FaderToDbCurve() -> (FaderToDbCurve:Curve Float Object Reference)
    InTime:float="To Float (Integer)"(InInt=@n1)
   ) -> (ReturnValue:float)
   n3 = float * float(A=@n2, B=0.05) {is=title K2Node_PromotableOperator}
   > Return Node!(
    execute
    Linear:double="Clamp (Float)"(Value=Power!(Base:double=10, Exp:double=@n3) -> (ReturnValue:double) {is=title K2Node_CallFunction}, Min=0, Max=1)
    Decibel:double="Min (Float)"(A:double=@n2, B:double=0) -> (ReturnValue:double) {is=title K2Node_CommutativeAssociativeBinaryOperator}
   )
  ]}
 FindInputDevices {nodes [
   FindInputDevices() {is=title K2Node_FunctionEntry}
   > n1 = Find MIDIDevices!(execute) -> (then, OutMIDIDevices:Array of Structures)
   > n2 = Branch(Condition="Equal (Integer)"(A=Length(TargetArray:Array of Structures=@n1)))
   > PrintString(InString="No Midi Devices Found :(", TextColor=(R=1,G=0.0305,B=0,A=1), Duration=10)
   n3 = For Each Loop!(Exec=@n2.else, Array:Array of Structures=@n1) -> (
    LoopBody
    Array Element:Structure
    Array Element_bIsAlreadyInUse:bool=~
    Array Element_bCanSendTo:bool=~
    Array Element_bCanReceiveFrom:bool=~
    Array Element_DeviceName:String
    Array Element_DeviceID:int
    Array Index
    Completed
   )
   > PrintString(
    InString="To String (Text)"(InText=Format Text!(
      Format="LOCTEXT(\"Midi Device {name} Id: {ID} IsInput {I} IsOutput {O} IsInUse {U}\")"
      name:String=@n3.Array Element_DeviceName
      ID:int=@n3.Array Element_DeviceID
      I:bool=@n3.Array Element_bCanReceiveFrom
      O:bool=@n3.Array Element_bCanSendTo
      U:bool=@n3.Array Element_bIsAlreadyInUse
     ) -> (Result))
    Duration=10
   )
   n4 = "Equal, Case Insensitive (String)"(A=@n3.Array Element_DeviceName, B=Get DeviceNameToUse() -> (DeviceNameToUse:String))
   > Branch(Condition=@n4)
   > n5 = Branch(Condition=@n3.Array Element_bCanReceiveFrom)
   > Set InDeviceID(InDeviceID:int=@n3.Array Element_DeviceID) -> (Output_Get:int=0)
   Set OutDeviceID(execute=@n5.else, OutDeviceID:int=@n3.Array Element_DeviceID) -> (Output_Get:int=0)
   AND Boolean(A=@n4, B=@n3.Array Element_bCanReceiveFrom)
  ]}
 GetCurrentWorld {nodes [
   GetCurrentWorld() {is=title K2Node_FunctionEntry}
   > n1 = GetPIEWorlds!(execute, bIncludeDedicatedServer:bool=false) -> (then, ReturnValue:Array of World Object References)
   > n2 = Branch(Condition=Is Valid Index!(TargetArray:Array of World Object References=@n1, IndexToTest:int) -> (ReturnValue:bool))
   > Return Node!(execute, ReturnValue:World Object Reference="Get (a copy)"(Array:Array of World Object References=@n1) -> (Output:World Object Reference))
   n3 = GetEditorWorld(execute=@n2.else, self=Get UnrealEditorSubsystem())
   > Return Node(ReturnValue=@n3)
  ]}
 InitializePhysicalFaders {nodes [
   InitializePhysicalFaders() {is=title K2Node_FunctionEntry}
   > n1 = Get!(execute:Exec) -> (then:Exec, else:Exec, BusMixToUse:Sound Control Bus Mix Object Reference)
   > n2 = Get!(execute) -> (then, else, MIDI Output Controller:Object Reference)
   n3 = Get MixStages!(self:Sound Control Bus Mix Object Reference=@n1) -> (MixStages:Array of Sound Control Bus Mix Stage Structures)
   n4 = Last Index!(TargetArray:Array of Sound Control Bus Mix Stage Structures=@n3) -> (ReturnValue=0)
   n5 = int * int(A=Get FaderShiftIdx() -> (FaderShiftIdx=0), B=8)
   > n6 = For Loop(LastIndex="Clamp (Integer)"(Value=int - int(A=@n4, B=@n5), Min=0, Max=7) -> (ReturnValue=0))
   n7 = Log!(
    A:double="Get (a ref)"!(
     Array:Array of Sound Control Bus Mix Stage Structures=@n3
     Dimension 1=FaderIdToMixStage!(A:int=@n6) -> (ReturnValue:int=0) {is=title K2Node_CallFunction}
    ) -> (
     Output_Bus:Sound Control Bus Object Reference
     Output_Value_TargetValue:float=~
     Output_Value_AttackTime:float=~
     Output_Value_ReleaseTime:float=~
    ).Output_Value_TargetValue
    Base:double=10
   ) -> (ReturnValue:double)
   n8 = Truncate(A=GetFloatValue(self=Get DbToFaderCurve() -> (DbToFaderCurve:Curve Float Object Reference), InTime=float * float(A=@n7, B=20)))
   > Send MIDIPitch Bend!(execute, self:Object Reference=@n2, Channel:int=int + int(A=@n6, B=~), Pitch:int=@n8) -> (then)
   > Send MIDINote On!(execute, self:Object Reference=Get MIDI Output Controller(), Channel:int=0, Note:int=int + int(A=8, B=@n6), Velocity:int=128) -> (then)
   > Branch(Condition=Get PrintPitchValues() -> (PrintPitchValues:bool))
   n9 = FaderIdToMixStage(A=@n6) -> (ReturnValue=0)
   > PrintString(InString="To String (Text)"(InText=Format Text!(Format="LOCTEXT(\"Fader Channel {C} udpated to Pitch {P}\")", C:int=@n9, P:int=@n8) -> (Result)))
  ]}
 OnFaderInput {nodes [
   n1 = OnFaderInput!() -> (then, Value:double=~, Channel:int) {is=title K2Node_FunctionEntry}
   n2 = Get BusMixToUse()
   n3 = Get MixStages(self=@n2)
   > Branch(Condition=Is Valid Index(TargetArray:Array of Sound Control Bus Mix Stage Structures=@n3, IndexToTest=@n1.Channel))
   > BeginTransaction(Context=ControlBusMixer, Description="LOCTEXT(\"Change Mix Stage Value\")", PrimaryObject=@n2)
   > TransactObject(Object=@n2)
   n4 = "Get (a ref)"(Array=@n3, Dimension 1=@n1.Channel) -> (Output_Value_TargetValue=~, Output_Value_AttackTime=~, Output_Value_ReleaseTime=~)
   > Set members in Sound Control Bus Mix Stage!(
    execute:Exec
    StructRef:"Sound Control Bus Mix Stage Structure (by ref)"="Get (a ref)"!(Array=@n3, Dimension 1=@n1.Channel) -> (Output:"Sound Control Bus Mix Stage Structure (by ref)")
    Value_TargetValue:float=@n1
    Value_AttackTime:float=@n4.Output_Value_AttackTime
    Value_ReleaseTime:float=@n4.Output_Value_ReleaseTime
   ) -> (then:Exec, StructOut:"Sound Control Bus Mix Stage Structure (by ref)") {is=title K2Node_SetFieldsInStruct}
   n5 = Get MixStages(self=@n2)
   n6 = Get BusMixToUse()
   > Set Control Bus Mix!(
    execute
    WorldContextObject:Object Reference=Get CurrentWorld()
    Mix:Sound Control Bus Mix Object Reference=@n6
    Stages:Array of Sound Control Bus Mix Stage Structures=@n5
    FadeTime:float=0
    Duration:double=-1
    bRetriggerOnActivation:bool=false
   ) -> (then) {is=title K2Node_CallFunction}
   > SetEditorProperty(Object=@n6, PropertyName=MixStages, PropertyValue:Array of Sound Control Bus Mix Stage Structures=@n5)
   > EndTransaction()
   > SetPhysicalFaderValue!(execute, self, Channel:int=ChannelToFaderId!(A:int=@n1.Channel) -> (ReturnValue:int=0) {is=title K2Node_CallFunction}) -> (then)
  ]}
 ResetAllFaders {nodes [
   ResetAllFaders() {is=title K2Node_FunctionEntry}
   > n1 = Get()
   > n2 = For Loop(LastIndex=7)
   > Send MIDIPitch Bend(self=@n1, Channel=@n2, Pitch=0)
   > Send MIDINote On(self=Get MIDI Output Controller(), Channel=0, Note=int + int(A=8, B=@n2), Velocity=0)
  ]}
 SaveControlBus {nodes [
   SaveControlBus() {is=title K2Node_FunctionEntry}
   n1 = Get BusMixToUse()
   > Is Valid(InputObject=@n1)
   > BeginTransaction(Context=AudioModulation, Description="LOCTEXT(\"Update Control Bus Mix\")", PrimaryObject=@n1)
   > TransactObject(Object=@n1)
   > n2 = SetEditorProperty(Object=@n1, PropertyName=MixStages, PropertyValue=Get MixStages(self=Get BusMixToUse()), ChangeNotifyMode=Always)
   > n3 = SaveLoadedAsset(AssetToSave=Get BusMixToUse(), bOnlyIfIsDirty=false)
   > PrintString(InString="To String (Text)"(InText=Format Text!(Format="LOCTEXT(\"Property Set {P} save successful {S}\")", P:bool=@n2, S:bool=@n3) -> (Result)))
   > EndTransaction()
   Get Object Path String(Object=@n1)
  ]}
 SetPhysicalFaderValue {nodes [
   n1 = SetPhysicalFaderValue!() -> (then, Channel:int) {is=title K2Node_FunctionEntry}
   > n2 = Get!(execute) -> (then, else, BusMixToUse:Sound Control Bus Mix Object Reference)
   > n3 = Get!(execute) -> (then, else, MIDI Output Controller:Object Reference)
   n4 = Log(
    A="Get (a ref)"!(Array=Get MixStages(self=@n2), Dimension 1=FaderIdToMixStage(A=@n1) -> (ReturnValue=0)) -> (
     Output_Bus:Sound Control Bus Object Reference
     Output_Value_TargetValue:float=~
     Output_Value_AttackTime:float=~
     Output_Value_ReleaseTime:float=~
    ).Output_Value_TargetValue
    Base=10
   )
   n5 = Truncate(A=GetFloatValue(self=Get DbToFaderCurve(), InTime=float * float(A=@n4, B=20)))
   > Send MIDIPitch Bend(self=@n3, Channel=@n1, Pitch=@n5)
   > Branch(Condition=Get PrintPitchValues())
   n6 = FaderIdToMixStage(A=0) -> (ReturnValue=0)
   > PrintString(InString="To String (Text)"(InText=Format Text!(Format="LOCTEXT(\"Fader Channel {C} udpated to Pitch {P}\")", C:int=@n6, P:int=@n5) -> (Result)))
  ]}
 SetUpFaders {nodes [
   SetUpFaders() {is=title K2Node_FunctionEntry}
   > n1 = For Each Loop!(Exec, Array:Array of WBP Control Bus Fader Object References=Get Faders() -> (Faders:Array of WBP Control Bus Fader Object References)) -> (LoopBody, Array Element:WBP Control Bus Fader Object Reference, Array Index, Completed)
   Clear(execute=@n1.Completed, TargetArray:Array of WBP Control Bus Fader Object References=Get Faders())
   > n2 = For Each Loop(Array:Array of Sound Control Bus Mix Stage Structures=Get MixStages(self=Get BusMixToUse())) -> (Array Element:Sound Control Bus Mix Stage Structure)
   n3 = Get FadersScrollBox() -> (FadersScrollBox:Scroll Box Object Reference)
   > n4 = Construct Object from Class(Class=&/WBP_ControlBusFader.WBP_ControlBusFader_C, self=@n3) -> (ReturnValue:WBP Control Bus Fader Object Reference)
   > AddChild(self=@n3, Content=@n4)
   > Add(TargetArray:Array of WBP Control Bus Fader Object References=Get Faders(), NewItem:"WBP Control Bus Fader Object Reference (by ref)"=@n4) {is=title K2Node_CallArrayFunction}
   n5 = Break Sound Control Bus Mix Stage!(SoundControlBusMixStage:"Sound Control Bus Mix Stage Structure (by ref)"=@n2) -> (Bus:Sound Control Bus Object Reference, Value_TargetValue:float=~, Value_AttackTime:float=~, Value_ReleaseTime:float=~) {is=title K2Node_BreakStruct}
   > SetFaderData!(
    execute
    self:WBP Control Bus Fader Object Reference=@n4
    FaderChannel:int=@n2.Array Index
    NormalizedValue:double=@n5.Value_TargetValue
    Fader Name:Text="To Text (String)"(InString=GetDisplayName(Object=@n5))
   ) -> (then)
   > Bind Event to On Fader Value Changed(self:WBP Control Bus Fader Object Reference=@n4, Delegate=Create Event()) {is=title K2Node_AddDelegate}
   RemoveFromParent(execute=@n1.LoopBody, self=@n1)
  ]}
 UpdateFader {nodes [
   n1 = UpdateFader!() -> (then, Channel:int, Pitch:int) {is=title K2Node_FunctionEntry}
   n2 = FaderIdToMixStage(A=@n1) -> (ReturnValue=0)
   n3 = int - int(A=@n2)
   n4 = Get MixStages(self=Get BusMixToUse())
   > Branch(Condition=Is Valid Index(TargetArray=@n4, IndexToTest=@n3))
   n5 = FaderPitchToLinearVolume!(InInt:int=@n1.Pitch) -> (Linear:double, Decibel:double) {is=title K2Node_CallFunction}
   n6 = "Get (a ref)"!(Array=@n4, Dimension 1=@n3) -> (Output:"Sound Control Bus Mix Stage Structure (by ref)")
   > Set members in Sound Control Bus Mix Stage(StructRef=@n6, Value_TargetValue=@n5, Value_AttackTime=0.1, Value_ReleaseTime=0.1)
   > n7 = Branch(Condition=Get PrintPitchValues())
   > n8 = PrintString(InString="To String (Text)"(InText=Format Text!(
      Format="LOCTEXT(\"MIDI Input on Channel {C}: Pitch Bend Value {P}, resulting volume {V} {dB} dB\")"
      C=@n2
      P=@n1.Pitch
      V:double=@n5
      dB:double=@n5.Decibel
     ) -> (Result)))
   n9 = Get BusMixToUse()
   BeginTransaction(execute=[@n8.then, @n7.else], Context=ControlBusMixer, Description="LOCTEXT(\"Change Mix Stage Value\")", PrimaryObject=@n9)
   > TransactObject(Object=@n9)
   n10 = Get BusMixToUse()
   > Set Control Bus Mix(WorldContextObject=Get CurrentWorld(), Mix=@n10, Stages=@n4, FadeTime=0, Duration=-1, bRetriggerOnActivation=false)
   > SetEditorProperty(Object=@n10, PropertyName=MixStages, PropertyValue=Get MixStages(self=@n9))
   > EndTransaction()
   n11 = Get Faders()
   > Branch(Condition=Is Valid Index(TargetArray:Array of WBP Control Bus Fader Object References=@n11, IndexToTest=@n3))
   n12 = Break Sound Control Bus Mix Stage(SoundControlBusMixStage=@n6) -> (Value_TargetValue=~, Value_AttackTime=~, Value_ReleaseTime=~)
   > SetFaderData(
    self="Get (a copy)"(Array:Array of WBP Control Bus Fader Object References=@n11, Dimension 1=@n3) -> (Output:WBP Control Bus Fader Object Reference)
    FaderChannel=@n3
    NormalizedValue=@n12.Value_TargetValue
    Fader Name="To Text (String)"(InString=GetDisplayName(Object=@n12))
   )
  ]}
 UpdateFaderShift {nodes [
   n1 = UpdateFaderShift!() -> (then, IsShiftUp:bool=~) {is=title K2Node_FunctionEntry}
   > n2 = Branch(Condition=@n1)
   n3 = Get FaderShiftIdx() -> (FaderShiftIdx=0)
   > n4 = Set FaderShiftIdx(FaderShiftIdx:int="Clamp (Integer)"(Value=int + int(A=@n3), Min=0, Max=4) -> (ReturnValue=0)) -> (Output_Get:int=0)
   n5 = int * int(A=Get FaderShiftIdx() -> (FaderShiftIdx=0), B=8)
   n6 = Get Faders()
   n7 = Branch(execute=[@n9.then, @n4.then], Condition=Is Valid Index(TargetArray=@n6, IndexToTest=@n5))
   > n8 = ScrollWidgetIntoView!(
    execute
    self:Scroll Box Object Reference=Get FadersScrollBox()
    WidgetToFind:Widget Object Reference="Get (a copy)"(Array=@n6, Dimension 1=@n5)
    AnimateScroll:bool=false
    ScrollDestination:EDescendantScrollDestination Enum=TopOrLeft
    Padding:float=0
   ) -> (then)
   ResetAllFaders(execute=[@n8.then, @n7.else]) {is=title K2Node_CallFunction}
   > InitializePhysicalFaders() {is=title K2Node_CallFunction}
   n9 = Set FaderShiftIdx(execute=@n2.else, FaderShiftIdx="Clamp (Integer)"(Value=int - int(A=@n3), Min=0, Max=4) -> (ReturnValue=0)) -> (Output_Get=0)
  ]}
]
variables [
 BusMixToUse {type=SoundControlBusMix, default=null}
 CurrentWorld {type=Object, default=null}
 DbToFaderCurve {type=CurveFloat, default=&/DbToFaderPortFader}
 DeviceNameToUse {type=string, default=PreSonus FP8}
 Faders {type="array<WBP_ControlBusFader_C>", default []}
 FaderShiftIdx {type=integer}
 FaderToDbCurve {type=CurveFloat, default=&/FaderPortFaderToDb}
 InDeviceID {type=integer}
 IsShiftedUp {type=boolean}
 MIDI Input Controller {type=Object, default=null}
 MIDI Output Controller {type=Object, default=null}
 OutDeviceID {type=integer}
 PrintPitchValues {type=boolean, default=true}
]
widgets {
 info {rootWidgetClass=&CanvasPanel, widgetCount=12}
 widgets [
  CanvasPanel_0 {widgetClassPath=&CanvasPanel, widget=CanvasPanel_0}
  HorizontalBox_42 {
   widgetClassPath=&HorizontalBox
   parent=CanvasPanel_0
   widget=HorizontalBox_42
   slot {class=CanvasPanelSlot, properties {layoutData {alignment=<0 0>, anchors {maximum=<1 1>, minimum=<0 0>}, offsets=<ltrb 0 0 0 0>}}}
  }
  VerticalBox_30 {widgetClassPath=&VerticalBox, parent=HorizontalBox_42, widget=VerticalBox_30, slot {class=HorizontalBoxSlot}}
  DeviceName_Property {
   widgetClassPath=&SinglePropertyView
   parent=VerticalBox_30
   widget=DeviceName_Property
   properties {propertyName=DeviceNameToUse}
   slot {class=VerticalBoxSlot}
  }
  ControlBusMix_Property {
   widgetClassPath=&SinglePropertyView
   parent=VerticalBox_30
   widget=ControlBusMix_Property
   properties {propertyName=BusMixToUse}
   slot {class=VerticalBoxSlot}
  }
  ScrollBox_40 {
   widgetClassPath=&ScrollBox
   bIsVariable=true
   parent=VerticalBox_30
   widget=ScrollBox_40
   properties {
    alwaysShowScrollbar=true
    consumeMouseWheel=Always
    widgetBarStyle {
     thickness=8
     draggedThumbImage {
      bIsDynamicallyLoaded=false
      drawAs=RoundedBox
      imageType=NoImage
      mirroring=NoMirror
      resourceName=None
      resourceObject=null
      tiling=NoTile
      imageSize=<0 0>
      margin=<ltrb 0 0 0 0>
      outlineSettings {
       bUseBrushTransparency=false
       roundingType=FixedRadius
       width=0
       color {colorUseRule=UseColor_Specified, specifiedColor=<rgba 0 0 0 0>}
       cornerRadii=<4 4 4 4>
      }
      tintColor {colorUseRule=UseColor_Specified, specifiedColor=<rgba 0.2159 0.2159 0.2159 1>}
      uVRegion {bIsValid=false, max=<0 0>, min=<0 0>}
     }
     horizontalBackgroundImage {
      bIsDynamicallyLoaded=false
      drawAs=NoDrawType
      imageType=NoImage
      mirroring=NoMirror
      resourceName=None
      resourceObject=null
      tiling=NoTile
      imageSize=<0 0>
      margin=<ltrb 0 0 0 0>
      outlineSettings {
       bUseBrushTransparency=false
       roundingType=HalfHeightRadius
       width=0
       color {colorUseRule=UseColor_Specified, specifiedColor=<rgba 0 0 0 0>}
       cornerRadii=<0 0 0 0>
      }
      tintColor {colorUseRule=UseColor_Specified, specifiedColor=<rgba 1 1 1 1>}
      uVRegion {bIsValid=false, max=<0 0>, min=<0 0>}
     }
     horizontalBottomSlotImage {
      bIsDynamicallyLoaded=false
      drawAs=NoDrawType
      imageType=NoImage
      mirroring=NoMirror
      resourceName=None
      resourceObject=null
      tiling=NoTile
      imageSize=<0 0>
      margin=<ltrb 0 0 0 0>
      outlineSettings {
       bUseBrushTransparency=false
       roundingType=HalfHeightRadius
       width=0
       color {colorUseRule=UseColor_Specified, specifiedColor=<rgba 0 0 0 0>}
       cornerRadii=<0 0 0 0>
      }
      tintColor {colorUseRule=UseColor_Specified, specifiedColor=<rgba 1 1 1 1>}
      uVRegion {bIsValid=false, max=<0 0>, min=<0 0>}
     }
     horizontalTopSlotImage {
      bIsDynamicallyLoaded=false
      drawAs=NoDrawType
      imageType=NoImage
      mirroring=NoMirror
      resourceName=None
      resourceObject=null
      tiling=NoTile
      imageSize=<0 0>
      margin=<ltrb 0 0 0 0>
      outlineSettings {
       bUseBrushTransparency=false
       roundingType=HalfHeightRadius
       width=0
       color {colorUseRule=UseColor_Specified, specifiedColor=<rgba 0 0 0 0>}
       cornerRadii=<0 0 0 0>
      }
      tintColor {colorUseRule=UseColor_Specified, specifiedColor=<rgba 1 1 1 1>}
      uVRegion {bIsValid=false, max=<0 0>, min=<0 0>}
     }
     hoveredThumbImage {
      bIsDynamicallyLoaded=false
      drawAs=RoundedBox
      imageType=NoImage
      mirroring=NoMirror
      resourceName=None
      resourceObject=null
      tiling=NoTile
      imageSize=<0 0>
      margin=<ltrb 0 0 0 0>
      outlineSettings {
       bUseBrushTransparency=false
       roundingType=FixedRadius
       width=0
       color {colorUseRule=UseColor_Specified, specifiedColor=<rgba 0 0 0 0>}
       cornerRadii=<4 4 4 4>
      }
      tintColor {colorUseRule=UseColor_Specified, specifiedColor=<rgba 0.2159 0.2159 0.2159 1>}
      uVRegion {bIsValid=false, max=<0 0>, min=<0 0>}
     }
     normalThumbImage {
      bIsDynamicallyLoaded=false
      drawAs=RoundedBox
      imageType=NoImage
      mirroring=NoMirror
      resourceName=None
      resourceObject=null
      tiling=NoTile
      imageSize=<0 0>
      margin=<ltrb 0 0 0 0>
      outlineSettings {
       bUseBrushTransparency=false
       roundingType=FixedRadius
       width=0
       color {colorUseRule=UseColor_Specified, specifiedColor=<rgba 0 0 0 0>}
       cornerRadii=<4 4 4 4>
      }
      tintColor {colorUseRule=UseColor_Specified, specifiedColor=<rgba 0.0953 0.0953 0.0953 1>}
      uVRegion {bIsValid=false, max=<0 0>, min=<0 0>}
     }
     verticalBackgroundImage {
      bIsDynamicallyLoaded=false
      drawAs=NoDrawType
      imageType=NoImage
      mirroring=NoMirror
      resourceName=None
      resourceObject=null
      tiling=NoTile
      imageSize=<0 0>
      margin=<ltrb 0 0 0 0>
      outlineSettings {
       bUseBrushTransparency=false
       roundingType=HalfHeightRadius
       width=0
       color {colorUseRule=UseColor_Specified, specifiedColor=<rgba 0 0 0 0>}
       cornerRadii=<0 0 0 0>
      }
      tintColor {colorUseRule=UseColor_Specified, specifiedColor=<rgba 1 1 1 1>}
      uVRegion {bIsValid=false, max=<0 0>, min=<0 0>}
     }
     verticalBottomSlotImage {
      bIsDynamicallyLoaded=false
      drawAs=NoDrawType
      imageType=NoImage
      mirroring=NoMirror
      resourceName=None
      resourceObject=null
      tiling=NoTile
      imageSize=<0 0>
      margin=<ltrb 0 0 0 0>
      outlineSettings {
       bUseBrushTransparency=false
       roundingType=HalfHeightRadius
       width=0
       color {colorUseRule=UseColor_Specified, specifiedColor=<rgba 0 0 0 0>}
       cornerRadii=<0 0 0 0>
      }
      tintColor {colorUseRule=UseColor_Specified, specifiedColor=<rgba 1 1 1 1>}
      uVRegion {bIsValid=false, max=<0 0>, min=<0 0>}
     }
     verticalTopSlotImage {
      bIsDynamicallyLoaded=false
      drawAs=NoDrawType
      imageType=NoImage
      mirroring=NoMirror
      resourceName=None
      resourceObject=null
      tiling=NoTile
      imageSize=<0 0>
      margin=<ltrb 0 0 0 0>
      outlineSettings {
       bUseBrushTransparency=false
       roundingType=HalfHeightRadius
       width=0
       color {colorUseRule=UseColor_Specified, specifiedColor=<rgba 0 0 0 0>}
       cornerRadii=<0 0 0 0>
      }
      tintColor {colorUseRule=UseColor_Specified, specifiedColor=<rgba 1 1 1 1>}
      uVRegion {bIsValid=false, max=<0 0>, min=<0 0>}
     }
    }
    widgetStyle {
     barThickness=8
     bottomShadowBrush {
      bIsDynamicallyLoaded=false
      drawAs=Box
      imageType=FullColor
      mirroring=NoMirror
      resourceName=../../../Engine/Content/Slate/Common/ScrollBoxShadowBottom.png
      resourceObject=null
      tiling=NoTile
      imageSize=<16 8>
      margin=<ltrb 0.5 0 0.5 1>
      outlineSettings {
       bUseBrushTransparency=false
       roundingType=HalfHeightRadius
       width=0
       color {colorUseRule=UseColor_Specified, specifiedColor=<rgba 0 0 0 0>}
       cornerRadii=<0 0 0 0>
      }
      tintColor {colorUseRule=UseColor_Specified, specifiedColor=<rgba 1 1 1 1>}
      uVRegion {bIsValid=false, max=<0 0>, min=<0 0>}
     }
     horizontalScrolledContentPadding=<ltrb 0 0 1 0>
     leftShadowBrush {
      bIsDynamicallyLoaded=false
      drawAs=Box
      imageType=FullColor
      mirroring=NoMirror
      resourceName=../../../Engine/Content/Slate/Common/ScrollBoxShadowLeft.png
      resourceObject=null
      tiling=NoTile
      imageSize=<8 16>
      margin=<ltrb 1 0.5 0 0.5>
      outlineSettings {
       bUseBrushTransparency=false
       roundingType=HalfHeightRadius
       width=0
       color {colorUseRule=UseColor_Specified, specifiedColor=<rgba 0 0 0 0>}
       cornerRadii=<0 0 0 0>
      }
      tintColor {colorUseRule=UseColor_Specified, specifiedColor=<rgba 1 1 1 1>}
      uVRegion {bIsValid=false, max=<0 0>, min=<0 0>}
     }
     rightShadowBrush {
      bIsDynamicallyLoaded=false
      drawAs=Box
      imageType=FullColor
      mirroring=NoMirror
      resourceName=../../../Engine/Content/Slate/Common/ScrollBoxShadowRight.png
      resourceObject=null
      tiling=NoTile
      imageSize=<8 16>
      margin=<ltrb 0 0.5 1 0.5>
      outlineSettings {
       bUseBrushTransparency=false
       roundingType=HalfHeightRadius
       width=0
       color {colorUseRule=UseColor_Specified, specifiedColor=<rgba 0 0 0 0>}
       cornerRadii=<0 0 0 0>
      }
      tintColor {colorUseRule=UseColor_Specified, specifiedColor=<rgba 1 1 1 1>}
      uVRegion {bIsValid=false, max=<0 0>, min=<0 0>}
     }
     topShadowBrush {
      bIsDynamicallyLoaded=false
      drawAs=Box
      imageType=FullColor
      mirroring=NoMirror
      resourceName=../../../Engine/Content/Slate/Common/ScrollBoxShadowTop.png
      resourceObject=null
      tiling=NoTile
      imageSize=<16 8>
      margin=<ltrb 0.5 1 0.5 0>
      outlineSettings {
       bUseBrushTransparency=false
       roundingType=HalfHeightRadius
       width=0
       color {colorUseRule=UseColor_Specified, specifiedColor=<rgba 0 0 0 0>}
       cornerRadii=<0 0 0 0>
      }
      tintColor {colorUseRule=UseColor_Specified, specifiedColor=<rgba 1 1 1 1>}
      uVRegion {bIsValid=false, max=<0 0>, min=<0 0>}
     }
     verticalScrolledContentPadding=<ltrb 0 0 0 1>
    }
   }
   slot {class=VerticalBoxSlot, properties {size {sizeRule=Fill, value=0.8}}}
  }
  DetailsView {widgetClassPath=&DetailsView, parent=ScrollBox_40, widget=DetailsView, slot {class=ScrollBoxSlot}}
  VerticalBox_317 {
   widgetClassPath=&VerticalBox
   parent=HorizontalBox_42
   widget=VerticalBox_317
   slot {class=HorizontalBoxSlot, properties {size {sizeRule=Fill, value=0.8}}}
  }
  Button_Refresh {
   widgetClassPath=&Button
   parent=VerticalBox_317
   widget=Button_Refresh
   properties {widgetStyle {
     clickedSlateSound {resourceObject=null}
     disabled {
      bIsDynamicallyLoaded=false
      drawAs=RoundedBox
      imageType=NoImage
      mirroring=NoMirror
      resourceName=None
      resourceObject=null
      tiling=NoTile
      imageSize=<0 0>
      margin=<ltrb 0 0 0 0>
      outlineSettings {
       bUseBrushTransparency=false
       roundingType=FixedRadius
       width=1
       color {colorUseRule=UseColor_ColorTable, specifiedColor=<rgba 0 0 0 0>}
       cornerRadii=<4 4 4 4>
      }
      tintColor {colorUseRule=UseColor_Specified, specifiedColor=<rgba 0.0395 0.0395 0.0395 1>}
      uVRegion {bIsValid=false, max=<0 0>, min=<0 0>}
     }
     disabledForeground {colorUseRule=UseColor_ColorTable, specifiedColor=<rgba 0 0 0 0>}
     hovered {
      bIsDynamicallyLoaded=false
      drawAs=RoundedBox
      imageType=NoImage
      mirroring=NoMirror
      resourceName=None
      resourceObject=null
      tiling=NoTile
      imageSize=<32 32>
      margin=<ltrb 0 0 0 0>
      outlineSettings {
       bUseBrushTransparency=false
       roundingType=FixedRadius
       width=1
       color {colorUseRule=UseColor_ColorTable, specifiedColor=<rgba 0 0 0 0>}
       cornerRadii=<4 4 4 4>
      }
      tintColor {colorUseRule=UseColor_Specified, specifiedColor=<rgba 0.0953 0.0953 0.0953 1>}
      uVRegion {bIsValid=false, max=<0 0>, min=<0 0>}
     }
     hoveredForeground {colorUseRule=UseColor_ColorTable, specifiedColor=<rgba 0 0 0 0>}
     hoveredSlateSound {resourceObject=null}
     normal {
      bIsDynamicallyLoaded=false
      drawAs=RoundedBox
      imageType=NoImage
      mirroring=NoMirror
      resourceName=None
      resourceObject=null
      tiling=NoTile
      imageSize=<32 32>
      margin=<ltrb 0 0 0 0>
      outlineSettings {
       bUseBrushTransparency=false
       roundingType=FixedRadius
       width=1
       color {colorUseRule=UseColor_ColorTable, specifiedColor=<rgba 0 0 0 0>}
       cornerRadii=<4 4 4 4>
      }
      tintColor {colorUseRule=UseColor_Specified, specifiedColor=<rgba 0.0395 0.0395 0.0395 1>}
      uVRegion {bIsValid=false, max=<0 0>, min=<0 0>}
     }
     normalForeground {colorUseRule=UseColor_ColorTable, specifiedColor=<rgba 0 0 0 0>}
     normalPadding=<ltrb 12 1.5 12 1.5>
     pressed {
      bIsDynamicallyLoaded=false
      drawAs=RoundedBox
      imageType=NoImage
      mirroring=NoMirror
      resourceName=None
      resourceObject=null
      tiling=NoTile
      imageSize=<32 32>
      margin=<ltrb 0 0 0 0>
      outlineSettings {
       bUseBrushTransparency=false
       roundingType=FixedRadius
       width=1
       color {colorUseRule=UseColor_ColorTable, specifiedColor=<rgba 0 0 0 0>}
       cornerRadii=<4 4 4 4>
      }
      tintColor {colorUseRule=UseColor_Specified, specifiedColor=<rgba 0.0284 0.0284 0.0284 1>}
      uVRegion {bIsValid=false, max=<0 0>, min=<0 0>}
     }
     pressedForeground {colorUseRule=UseColor_ColorTable, specifiedColor=<rgba 0 0 0 0>}
     pressedPadding=<ltrb 12 2.5 12 0.5>
     pressedSlateSound {resourceObject=null}
    }}
   slot {class=VerticalBoxSlot}
  }
  RichTextBlock_22 {
   widgetClassPath=/Script/UMG.RichTextBlock
   bIsVariable=true
   parent=Button_Refresh
   widget=RichTextBlock_22
   properties {text=Refresh, textStyleSet=/Game/Athena/HUD/Abilities/AbilityRichText}
   slot {class=ButtonSlot}
  }
  FaderBox {widgetClassPath=&HorizontalBox, bIsVariable=true, parent=VerticalBox_317, widget=FaderBox, slot {class=VerticalBoxSlot}}
  FadersScrollBox {
   widgetClassPath=&ScrollBox
   bIsVariable=true
   parent=FaderBox
   widget=FadersScrollBox
   properties {
    alwaysShowScrollbar=true
    alwaysShowScrollbarTrack=true
    consumeMouseWheel=Always
    orientation=Orient_Horizontal
    widgetBarStyle {
     thickness=8
     draggedThumbImage {
      bIsDynamicallyLoaded=false
      drawAs=RoundedBox
      imageType=NoImage
      mirroring=NoMirror
      resourceName=None
      resourceObject=null
      tiling=NoTile
      imageSize=<0 0>
      margin=<ltrb 0 0 0 0>
      outlineSettings {
       bUseBrushTransparency=false
       roundingType=FixedRadius
       width=0
       color {colorUseRule=UseColor_Specified, specifiedColor=<rgba 0 0 0 0>}
       cornerRadii=<4 4 4 4>
      }
      tintColor {colorUseRule=UseColor_Specified, specifiedColor=<rgba 0.2159 0.2159 0.2159 1>}
      uVRegion {bIsValid=false, max=<0 0>, min=<0 0>}
     }
     horizontalBackgroundImage {
      bIsDynamicallyLoaded=false
      drawAs=NoDrawType
      imageType=NoImage
      mirroring=NoMirror
      resourceName=None
      resourceObject=null
      tiling=NoTile
      imageSize=<0 0>
      margin=<ltrb 0 0 0 0>
      outlineSettings {
       bUseBrushTransparency=false
       roundingType=HalfHeightRadius
       width=0
       color {colorUseRule=UseColor_Specified, specifiedColor=<rgba 0 0 0 0>}
       cornerRadii=<0 0 0 0>
      }
      tintColor {colorUseRule=UseColor_Specified, specifiedColor=<rgba 1 1 1 1>}
      uVRegion {bIsValid=false, max=<0 0>, min=<0 0>}
     }
     horizontalBottomSlotImage {
      bIsDynamicallyLoaded=false
      drawAs=NoDrawType
      imageType=NoImage
      mirroring=NoMirror
      resourceName=None
      resourceObject=null
      tiling=NoTile
      imageSize=<0 0>
      margin=<ltrb 0 0 0 0>
      outlineSettings {
       bUseBrushTransparency=false
       roundingType=HalfHeightRadius
       width=0
       color {colorUseRule=UseColor_Specified, specifiedColor=<rgba 0 0 0 0>}
       cornerRadii=<0 0 0 0>
      }
      tintColor {colorUseRule=UseColor_Specified, specifiedColor=<rgba 1 1 1 1>}
      uVRegion {bIsValid=false, max=<0 0>, min=<0 0>}
     }
     horizontalTopSlotImage {
      bIsDynamicallyLoaded=false
      drawAs=NoDrawType
      imageType=NoImage
      mirroring=NoMirror
      resourceName=None
      resourceObject=null
      tiling=NoTile
      imageSize=<0 0>
      margin=<ltrb 0 0 0 0>
      outlineSettings {
       bUseBrushTransparency=false
       roundingType=HalfHeightRadius
       width=0
       color {colorUseRule=UseColor_Specified, specifiedColor=<rgba 0 0 0 0>}
       cornerRadii=<0 0 0 0>
      }
      tintColor {colorUseRule=UseColor_Specified, specifiedColor=<rgba 1 1 1 1>}
      uVRegion {bIsValid=false, max=<0 0>, min=<0 0>}
     }
     hoveredThumbImage {
      bIsDynamicallyLoaded=false
      drawAs=RoundedBox
      imageType=NoImage
      mirroring=NoMirror
      resourceName=None
      resourceObject=null
      tiling=NoTile
      imageSize=<0 0>
      margin=<ltrb 0 0 0 0>
      outlineSettings {
       bUseBrushTransparency=false
       roundingType=FixedRadius
       width=0
       color {colorUseRule=UseColor_Specified, specifiedColor=<rgba 0 0 0 0>}
       cornerRadii=<4 4 4 4>
      }
      tintColor {colorUseRule=UseColor_Specified, specifiedColor=<rgba 0.2159 0.2159 0.2159 1>}
      uVRegion {bIsValid=false, max=<0 0>, min=<0 0>}
     }
     normalThumbImage {
      bIsDynamicallyLoaded=false
      drawAs=RoundedBox
      imageType=NoImage
      mirroring=NoMirror
      resourceName=None
      resourceObject=null
      tiling=NoTile
      imageSize=<0 0>
      margin=<ltrb 0 0 0 0>
      outlineSettings {
       bUseBrushTransparency=false
       roundingType=FixedRadius
       width=0
       color {colorUseRule=UseColor_Specified, specifiedColor=<rgba 0 0 0 0>}
       cornerRadii=<4 4 4 4>
      }
      tintColor {colorUseRule=UseColor_Specified, specifiedColor=<rgba 0.0953 0.0953 0.0953 1>}
      uVRegion {bIsValid=false, max=<0 0>, min=<0 0>}
     }
     verticalBackgroundImage {
      bIsDynamicallyLoaded=false
      drawAs=NoDrawType
      imageType=NoImage
      mirroring=NoMirror
      resourceName=None
      resourceObject=null
      tiling=NoTile
      imageSize=<0 0>
      margin=<ltrb 0 0 0 0>
      outlineSettings {
       bUseBrushTransparency=false
       roundingType=HalfHeightRadius
       width=0
       color {colorUseRule=UseColor_Specified, specifiedColor=<rgba 0 0 0 0>}
       cornerRadii=<0 0 0 0>
      }
      tintColor {colorUseRule=UseColor_Specified, specifiedColor=<rgba 1 1 1 1>}
      uVRegion {bIsValid=false, max=<0 0>, min=<0 0>}
     }
     verticalBottomSlotImage {
      bIsDynamicallyLoaded=false
      drawAs=NoDrawType
      imageType=NoImage
      mirroring=NoMirror
      resourceName=None
      resourceObject=null
      tiling=NoTile
      imageSize=<0 0>
      margin=<ltrb 0 0 0 0>
      outlineSettings {
       bUseBrushTransparency=false
       roundingType=HalfHeightRadius
       width=0
       color {colorUseRule=UseColor_Specified, specifiedColor=<rgba 0 0 0 0>}
       cornerRadii=<0 0 0 0>
      }
      tintColor {colorUseRule=UseColor_Specified, specifiedColor=<rgba 1 1 1 1>}
      uVRegion {bIsValid=false, max=<0 0>, min=<0 0>}
     }
     verticalTopSlotImage {
      bIsDynamicallyLoaded=false
      drawAs=NoDrawType
      imageType=NoImage
      mirroring=NoMirror
      resourceName=None
      resourceObject=null
      tiling=NoTile
      imageSize=<0 0>
      margin=<ltrb 0 0 0 0>
      outlineSettings {
       bUseBrushTransparency=false
       roundingType=HalfHeightRadius
       width=0
       color {colorUseRule=UseColor_Specified, specifiedColor=<rgba 0 0 0 0>}
       cornerRadii=<0 0 0 0>
      }
      tintColor {colorUseRule=UseColor_Specified, specifiedColor=<rgba 1 1 1 1>}
      uVRegion {bIsValid=false, max=<0 0>, min=<0 0>}
     }
    }
    widgetStyle {
     barThickness=8
     bottomShadowBrush {
      bIsDynamicallyLoaded=false
      drawAs=Box
      imageType=FullColor
      mirroring=NoMirror
      resourceName=../../../Engine/Content/Slate/Common/ScrollBoxShadowBottom.png
      resourceObject=null
      tiling=NoTile
      imageSize=<16 8>
      margin=<ltrb 0.5 0 0.5 1>
      outlineSettings {
       bUseBrushTransparency=false
       roundingType=HalfHeightRadius
       width=0
       color {colorUseRule=UseColor_Specified, specifiedColor=<rgba 0 0 0 0>}
       cornerRadii=<0 0 0 0>
      }
      tintColor {colorUseRule=UseColor_Specified, specifiedColor=<rgba 1 1 1 1>}
      uVRegion {bIsValid=false, max=<0 0>, min=<0 0>}
     }
     horizontalScrolledContentPadding=<ltrb 0 0 1 0>
     leftShadowBrush {
      bIsDynamicallyLoaded=false
      drawAs=Box
      imageType=FullColor
      mirroring=NoMirror
      resourceName=../../../Engine/Content/Slate/Common/ScrollBoxShadowLeft.png
      resourceObject=null
      tiling=NoTile
      imageSize=<8 16>
      margin=<ltrb 1 0.5 0 0.5>
      outlineSettings {
       bUseBrushTransparency=false
       roundingType=HalfHeightRadius
       width=0
       color {colorUseRule=UseColor_Specified, specifiedColor=<rgba 0 0 0 0>}
       cornerRadii=<0 0 0 0>
      }
      tintColor {colorUseRule=UseColor_Specified, specifiedColor=<rgba 1 1 1 1>}
      uVRegion {bIsValid=false, max=<0 0>, min=<0 0>}
     }
     rightShadowBrush {
      bIsDynamicallyLoaded=false
      drawAs=Box
      imageType=FullColor
      mirroring=NoMirror
      resourceName=../../../Engine/Content/Slate/Common/ScrollBoxShadowRight.png
      resourceObject=null
      tiling=NoTile
      imageSize=<8 16>
      margin=<ltrb 0 0.5 1 0.5>
      outlineSettings {
       bUseBrushTransparency=false
       roundingType=HalfHeightRadius
       width=0
       color {colorUseRule=UseColor_Specified, specifiedColor=<rgba 0 0 0 0>}
       cornerRadii=<0 0 0 0>
      }
      tintColor {colorUseRule=UseColor_Specified, specifiedColor=<rgba 1 1 1 1>}
      uVRegion {bIsValid=false, max=<0 0>, min=<0 0>}
     }
     topShadowBrush {
      bIsDynamicallyLoaded=false
      drawAs=Box
      imageType=FullColor
      mirroring=NoMirror
      resourceName=../../../Engine/Content/Slate/Common/ScrollBoxShadowTop.png
      resourceObject=null
      tiling=NoTile
      imageSize=<16 8>
      margin=<ltrb 0.5 1 0.5 0>
      outlineSettings {
       bUseBrushTransparency=false
       roundingType=HalfHeightRadius
       width=0
       color {colorUseRule=UseColor_Specified, specifiedColor=<rgba 0 0 0 0>}
       cornerRadii=<0 0 0 0>
      }
      tintColor {colorUseRule=UseColor_Specified, specifiedColor=<rgba 1 1 1 1>}
      uVRegion {bIsValid=false, max=<0 0>, min=<0 0>}
     }
     verticalScrolledContentPadding=<ltrb 0 0 0 1>
    }
   }
   slot {class=HorizontalBoxSlot, properties {size {sizeRule=Fill, value=0.8}}}
  }
 ]
}
